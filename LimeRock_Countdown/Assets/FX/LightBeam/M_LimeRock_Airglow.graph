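# Additive unlit airglow material for LimeRock_Countdown light-beam FX, drawing a soft radial glow on a camera-facing card.
# World position offset rebuilds the card around the object position, growing it with camera distance and applying camera offset.
# Tinted emissive pairs with exponential-gradient opacity, optionally warped by panning Perlin noise and faded near the camera, plus depth fade under deferred shading.
# A separate polar-UV tech-texture branch feeds none of these chains.
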Material /LimeRock_Countdown/Assets/FX/LightBeam/M_LimeRock_Airglow
graph {
 properties {bAutomaticallySetUsageInEditor=false, blendMode=BLEND_Additive, shadingModel=MSM_Unlit, usedShadingModels=Unlit}
 nodes [
  n1 = ConstantBiasScale(None=TextureCoordinate()) {bias=-0.5, scale=1}
  n2 = ObjectPositionWS()
  n3 = CameraPositionWS()
  n4 = MaterialFunctionCall!(Vector1=Normalize(VectorInput=Subtract(A=@n2, B=@n3)), Vector2=Constant3Vector() {constant=<rgba 0 0 1 0>}) {materialFunction=/Engine/Functions/Engine_MaterialFunctions02/Math/CreateThirdOrthogonalVector}
  n5 = Normalize(VectorInput=@n4.Vector3)
  n6 = ScalarParameter() {defaultValue=1, parameterName=WPO Scale X}
  n7 = ScalarParameter() {defaultValue=1, parameterName=WPO Scale Y}
  n8 = ScalarParameter() {defaultValue=1, parameterName=WPO Scale Z}
  n9 = Add(A=Distance(A=ObjectPositionWS(), B=CameraPositionWS()), B=ScalarParameter() {parameterName=Scale Distance Offset})
  n10 = ScalarParameter() {defaultValue=2000, parameterName=Scale Distance Max}
  n11 = LinearInterpolate(
   B=StaticSwitchParameter(
    True=ScalarParameter() {defaultValue=1, parameterName=WPO Uniform Scale}
    False=MaterialFunctionCall(R=@n6, G=@n7, B=@n8) {materialFunction=&AppendMany}.RGB
   ) {defaultValue=true, parameterName=Use Uniform Scale}
   Alpha=Clamp(None=Divide(A=@n9, B=@n10))
  )
  n12 = ComponentMask(None=@n1) {g=true}
  n13 = Add(A=Multiply(A=ComponentMask(None=@n1) {r=true}, B=Multiply(A=@n5, B=@n11)), B=Multiply(A=Multiply(A=@n4.Vector2, B=@n12), B=@n11))
  n14 = WorldPosition()
  Add(
   A=Add(A=Subtract(A=@n13, B=@n14.XYZ), B=ObjectPositionWS())
   B=MaterialFunctionCall(Offset Amount=ScalarParameter() {parameterName=camera_Offset}) {materialFunction=/Engine/Functions/Engine_MaterialFunctions02/WorldPositionOffset/CameraOffset}.Clamped Camera Offset
  )
  MaterialFunctionCall(Emissive=Multiply(
    A=VectorParameter() {parameterName=Color, defaultValue=<rgba 1 1 1 1>}.RGB
    B=ScalarParameter() {defaultValue=1, parameterName=Emissive Intensity}
   )) {materialFunction=&MF_FX_EmissiveAdjustments}
  n15 = ScalarParameter() {defaultValue=1, parameterName=Opacity}
  n16 = OneMinus(None=SphereMask(
    A=ObjectPositionWS()
    B=CameraPositionWS()
    Radius=ScalarParameter() {defaultValue=1, group=Near Fade, parameterName=Fade_Distance}
    Hardness=ScalarParameter() {group=Near Fade, parameterName=hardness}
   ) {hardnessPercent=10})
  n17 = Constant() {r=1}
  n18 = DepthFade(
   Opacity=MaterialFunctionCall(Distance=ScalarParameter() {parameterName=NearCamFade}) {materialFunction=&Near_Camera_Fade}.Result
   FadeDistance=ScalarParameter() {defaultValue=100, parameterName=DepthFade}
  )
  n19 = ScalarParameter() {defaultValue=2.33, parameterName=Base Gradient Density}
  n20 = TextureCoordinate()
  n21 = Multiply(A=AppendVector(A=ScalarParameter() {parameterName=Pan Speed X}, B=ScalarParameter() {parameterName=Pan Speed Y}), B=Time())
  n22 = ScalarParameter() {defaultValue=1, parameterName=Tiling X}
  n23 = ScalarParameter() {defaultValue=1, parameterName=Tiling Y}
  n24 = TextureSample(UVs=Multiply(A=Add(A=@n20, B=@n21), B=AppendVector(A=@n22, B=@n23))) {samplerType=SAMPLERTYPE_Masks, texture=/Game/Characters/Player/Male/Large/Bodies/M_LRG_Clash_V/Textures/Simple_Perlin_M}
  n25 = Add(A=TextureCoordinate(), B=Multiply(A=Subtract(A=@n24.R) {constB=0.5}, B=ScalarParameter() {defaultValue=0.2, parameterName=GradientWarp}))
  Multiply(
   A=Multiply(A=Multiply(A=@n15, B=@n16), B=ShadingPathSwitch(Default=@n17, Deferred=@n18))
   B=LinearInterpolate(
    A=MaterialFunctionCall(Density=@n19) {materialFunction=&RadialGradientExponential}.RadialGradientExponential
    B=MaterialFunctionCall(UVs=@n25, Density=@n19) {materialFunction=&RadialGradientExponential}.RadialGradientExponential
    Alpha=ScalarParameter() {parameterName=Airglow Noise}
   )
  )
  n26 = VectorParameter() {parameterName=Main_UV, defaultValue=<rgba 3 2.101 -0.1307 0.248>}
  NamedRerouteDeclaration(None=MaterialFunctionCall!(
    ClampTexture
    UV=MaterialFunctionCall() {materialFunction=&VectorToRadialValue}.Radial Coordinates
    Tex
    InRGB=@n26.RGB
    InA=@n26.RGB
    Time
    TimeMult
   ) {materialFunction=/Hermes/Props/ClayTile/ClayTile_DragonGate_B/VFX/MF_MOON_TexPanner_V2}.PannerUV) {name=MainUV, nodeColor=<rgba 0 0.4049 1 1>}
  n27 = MaterialFunctionCall(Float2=NamedRerouteUsage()) {materialFunction=&BreakOutFloat2Components}
  n28 = Multiply(A=@n27.R, B=ScalarParameter() {defaultValue=1, parameterName=UV_Tilt})
  n29 = TextureSampleParameter2D!(
   UVs=MaterialFunctionCall(X=@n27.R, Y=Add(A=@n27.G, B=@n28)) {materialFunction=&MakeFloat2}.Result
   Level=ScalarParameter() {defaultValue=1.5, parameterName=Param_6}
   Apply View MipBias
  ) {
   mipValueMode=TMVM_MipLevel
   parameterName=Param
   samplerType=SAMPLERTYPE_LinearColor
   texture=/Hermes/Props/ClayTile/ClayTile_DragonGate_B/VFX/Tile_Tech_02
  }
  SquareRoot(None=@n29.G)
 ]
}
