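# Unlit additive material for a glowing light beam effect in LimeRock Countdown, blending glow and core colors through an inner-core mask.
# The core combines Fresnel and a height gradient eroded by panning, warped smoke noise.
# Opacity multiplies upper and lower bounds-based gradients, Fresnel with adjustable vertical falloff, deferred-only depth fade, and camera-proximity fades.
# A separate UV sphere mask, with optional Y clamping, has no visible consumer.
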Material /LimeRock_Countdown/Assets/FX/LightBeam/M_LimeRock_LightBeam
graph {
 properties {blendMode=BLEND_Additive, pixelDepthOffsetMode=PDOM_Legacy, shadingModel=MSM_Unlit, usedShadingModels=Unlit}
 nodes [
  LinearInterpolate(
   A=Multiply(
    A=VectorParameter() {parameterName=GlowColor, sortPriority=1, defaultValue=<rgba 9 3.7221 0 3>}.RGB
    B=ScalarParameter() {defaultValue=1, parameterName=Emissive Intensity, sortPriority=2}
   )
   B=VectorParameter() {parameterName=GlowColor_Core, sortPriority=1, defaultValue=<rgba 9 3.7221 0 3>}.RGB
   Alpha=NamedRerouteUsage()
  )
  n1 = ScalarParameter() {defaultValue=1, group=Inner Core, parameterName=Edge Fresnel Core, sortPriority=8}
  n2 = ScalarParameter() {defaultValue=1, group=Inner Core, parameterName=Edge Fresnel Core Exp, sortPriority=9}
  n3 = MaterialFunctionCall(Float3=CameraVectorWS()) {materialFunction=&BreakOutFloat3Components}
  n4 = ComponentMask(None=VertexNormalWS()) {b=true}
  n5 = Normalize(VectorInput=LinearInterpolate(
    A=CameraVectorWS()
    B=MaterialFunctionCall(X=@n3.R, Y=@n3.R, Z=@n4) {materialFunction=&MakeFloat3}.Result
    Alpha=ScalarParameter() {parameterName=Fresnel Verticle Falloff}
   ))
  n6 = TransformPosition(None=WorldPosition().XYZ) {transformSourceType=TRANSFORMPOSSOURCE_World}
  n7 = TransformPosition(None=ObjectPositionWS()) {transformSourceType=TRANSFORMPOSSOURCE_World}
  n8 = Add(A=Divide(A=Subtract(A=@n6, B=@n7), B=ObjectLocalBounds().Extents)) {constB=0.5}
  n9 = OneMinus(None=ComponentMask(None=Saturate(None=@n8)) {b=true})
  n10 = ScalarParameter() {defaultValue=2, group=Inner Core, parameterName=Upper Edge Core Gradient Pwr, sortPriority=3}
  n11 = TextureCoordinate()
  n12 = ComponentMask(None=VectorParameter() {group=Inner Core, parameterName=NoiseTiling, defaultValue=<rgba 1 5 0 0>}.RGB) {g=true, r=true}
  n13 = ScalarParameter() {defaultValue=-0.05, group=Inner Core, parameterName=SpeedX}
  n14 = ScalarParameter() {defaultValue=-0.25, group=Inner Core, parameterName=SpeedY}
  n15 = TextureCoordinate()
  n16 = ComponentMask(None=VectorParameter() {group=Inner Core, parameterName=Warp_NoiseTiling, defaultValue=<rgba 1 5 0 0>}.RGB) {g=true, r=true}
  n17 = ScalarParameter() {defaultValue=-0.05, group=Inner Core, parameterName=Warp_SpeedX}
  n18 = ScalarParameter() {defaultValue=-0.25, group=Inner Core, parameterName=Warp_SpeedY}
  n19 = TextureSample(UVs=Panner(Coordinate=Multiply(A=@n15, B=@n16), Speed=AppendVector(A=@n17, B=@n18)) {speedX=-0.05, speedY=-0.25}) {samplerType=SAMPLERTYPE_LinearColor, texture=/Game/Global/Textures/Default/Caustics/03_WaterFoam_AccentSwirls_H}
  n20 = Add(
   A=Panner(Coordinate=Multiply(A=@n11, B=@n12), Speed=AppendVector(A=@n13, B=@n14)) {speedX=-0.05, speedY=-0.25}
   B=Multiply(A=ComponentMask(None=@n19.RGB) {g=true, r=true}, B=ScalarParameter() {group=Inner Core, parameterName=Warp_Str})
  )
  n21 = Subtract(
   A=Multiply(
    A=Power(Base=@n9, Exp=@n10)
    B=ScalarParameter() {defaultValue=1, group=Inner Core, parameterName=Upper Edge Core Gradient Mult, sortPriority=4}
   )
   B=Multiply(
    A=TextureSample(UVs=@n20) {texture=&T_Inky_Smoke_Tile_Chromatic}.RGB
    B=ScalarParameter() {defaultValue=1, group=Inner Core, parameterName=NoiseOpacity}
   )
  )
  NamedRerouteDeclaration(None=Multiply(
    A=MaterialFunctionCall(Brightness=@n1, Exponent=@n2, Camera Vector=@n5) {materialFunction=&MF_Fresnel_FX}.Fresnel
    B=Saturate(None=@n21)
   )) {name=InnerCore, nodeColor=<rgba 0.6406 1 0 1>}
  n22 = WorldPosition()
  n23 = ObjectPositionWS()
  n24 = Divide(
   A=Subtract(
    A=TransformPosition(None=@n22.XYZ) {transformSourceType=TRANSFORMPOSSOURCE_World}
    B=TransformPosition(None=@n23) {transformSourceType=TRANSFORMPOSSOURCE_World}
   )
   B=ObjectLocalBounds().Extents
  )
  n25 = ComponentMask(None=Saturate(None=Add(A=@n24) {constB=0.5})) {b=true}
  n26 = Multiply(
   A=Power(Base=OneMinus(None=@n25), Exp=ScalarParameter() {defaultValue=2, parameterName=Upper Edge Gradient Pwr, sortPriority=3})
   B=ScalarParameter() {defaultValue=1, parameterName=Upper Edge Gradient Mult, sortPriority=4}
  )
  n27 = Multiply(
   A=Power(Base=@n25, Exp=ScalarParameter() {defaultValue=1, parameterName=Lower Edge Gradient Pwr, sortPriority=5})
   B=ScalarParameter() {defaultValue=1, parameterName=Lower Edge Gradient Mult, sortPriority=6}
  )
  n28 = ScalarParameter() {defaultValue=1, parameterName=Edge Fresnel Brightness, sortPriority=8}
  n29 = ScalarParameter() {defaultValue=1, parameterName=Edge Fresnel Exp, sortPriority=9}
  n30 = MaterialFunctionCall(Float3=CameraVectorWS()) {materialFunction=&BreakOutFloat3Components}
  n31 = ComponentMask(None=VertexNormalWS()) {b=true}
  n32 = Normalize(VectorInput=LinearInterpolate(
    A=CameraVectorWS()
    B=MaterialFunctionCall(X=@n30.R, Y=@n30.R, Z=@n31) {materialFunction=&MakeFloat3}.Result
    Alpha=ScalarParameter() {parameterName=Fresnel Verticle Falloff}
   ))
  n33 = ScalarParameter() {defaultValue=3, parameterName=Opacity, sortPriority=15}
  n34 = ScalarParameter() {parameterName=NearCamFade, sortPriority=16}
  n35 = ShadingPathSwitch(
   Default=Constant() {r=1}
   Deferred=DepthFade(
    Opacity=MaterialFunctionCall(Distance=@n34) {materialFunction=&Near_Camera_Fade}.Result
    FadeDistance=ScalarParameter() {defaultValue=100, parameterName=DepthFade, sortPriority=17}
   )
  )
  n36 = Multiply(
   A=Multiply(A=Saturate(None=@n26), B=Saturate(None=@n27))
   B=Multiply(
    A=MaterialFunctionCall(Brightness=@n28, Exponent=@n29, Camera Vector=@n32) {materialFunction=&MF_Fresnel_FX}.Fresnel
    B=Multiply(A=@n33, B=@n35)
   )
  )
  n37 = OneMinus(None=SphereMask(
    A=ObjectPositionWS()
    B=CameraPositionWS()
    Radius=ScalarParameter() {defaultValue=1, group=Near Fade, parameterName=Fade_Distance}
    Hardness=ScalarParameter() {group=Near Fade, parameterName=hardness}
   ) {hardnessPercent=10})
  Reroute(None=Saturate(None=Multiply(A=@n36, B=@n37))) {desc=Only for translucent}
  n38 = Multiply(
   A=TextureCoordinate()
   B=AppendVector(
    A=ScalarParameter() {defaultValue=1, group=SphereMask, parameterName=SphereMask-UV-X}
    B=ScalarParameter() {defaultValue=0.3, group=SphereMask, parameterName=SphereMask-UV-Y}
   )
  )
  n39 = ComponentMask(None=@n38) {g=true}
  n40 = ScalarParameter() {group=SphereMask, parameterName=Spheremask UV.Y Min}
  n41 = ScalarParameter() {defaultValue=0.5, group=SphereMask, parameterName=Spheremask UV.Y Max}
  n42 = StaticSwitchParameter(True=AppendVector(A=ComponentMask(None=@n38) {r=true}, B=Clamp(None=@n39, Min=@n40, Max=@n41) {maxDefault=0.5}), False=@n38) {
   desc=Clamps spheremask uv Y component. Use to make spheremask only affect one portion of the length of the beam
   group=SphereMask
   parameterName=Clamp Spheremask UV.Y
  }
  SphereMask(
   A=@n42
   B=AppendVector(
    A=ScalarParameter() {defaultValue=0.5, group=SphereMask, parameterName=SphereMask-CenterX}
    B=ScalarParameter() {defaultValue=-0.225, group=SphereMask, parameterName=SphereMask-CenterY}
   )
   Radius=ScalarParameter() {defaultValue=0.55, group=SphereMask, parameterName=SphereMask-Radius}
   Hardness=ScalarParameter() {group=SphereMask, parameterName=SphereMask-Hardness}
  )
 ]
}
