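# Translucent, unlit, two-sided energy beam material for thickened spline meshes, usable on Niagara ribbons and mesh particles.
# Panning plasma-coil and smoke textures, distorted by scrolling inky smoke, build tinted color layers and a noisy opacity shaped by a sine along U.
# Opacity fades by view angle, camera proximity and, on deferred shading, scene depth.
# SplineThicken offsets combine with timed axis rotation; a separate named reroute computes camera-facing rotation.
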Material /LimeRock_Countdown/Assets/FX/LightBeam/M_LimeRock_SplineThicken_Energy
graph {
 properties {
  bUsedWithNiagaraMeshParticles=true
  bUsedWithNiagaraRibbons=true
  blendMode=BLEND_Translucent
  pixelDepthOffsetMode=PDOM_Legacy
  shadingModel=MSM_Unlit
  twoSided=true
  usedShadingModels=Unlit
 }
 nodes [
  n1 = Add(A=TextureCoordinate() {vTiling=0.5}, B=Constant2Vector() {r=0.7})
  n2 = Multiply(A=Time(), B=ScalarParameter() {defaultValue=-1.5, parameterName=Speed Direction}) {constB=-1.5}
  n3 = Power(Base=TextureSample(UVs=Panner(Coordinate=@n1, Time=@n2) {speedX=-2}) {texture=&T_Tile_RainbowColor}.G)
  n4 = VectorParameter() {parameterName=Color, defaultValue=<rgba 1.5 1.4799 0.5675 1>}
  n5 = ScalarParameter() {parameterName=DefaultColorDes}
  n6 = TextureCoordinate() {uTiling=0.5, vTiling=0.5}
  n7 = TextureCoordinate()
  n8 = Multiply(A=TextureSample(UVs=Panner(Coordinate=@n7) {speedX=0.1}) {texture=&T_Inky_Smoke_Tile}.R) {constB=0.1}
  n9 = TextureSample(UVs=Panner(Coordinate=Add(A=@n8, B=@n6), Time=@n2) {speedX=-0.8}) {samplerType=SAMPLERTYPE_LinearGrayscale, texture=/Game/Effects/Fort_Effects/Emotes/UltraLight/Texture/T_PlasmaCoils}
  n10 = TextureCoordinate()
  n11 = TextureSample(UVs=Panner(Coordinate=Add(A=@n10, B=@n8), Time=@n2) {speedX=-1}) {samplerType=SAMPLERTYPE_Grayscale, texture=&T_Smoke_Trail_Soft}
  n12 = Add(
   A=Multiply(A=Desaturation(None=@n4.RGB, Fraction=@n5), B=Power(Base=@n9.R) {constExponent=1.5})
   B=Multiply(A=VectorParameter() {parameterName=ColorDark, defaultValue=<rgba 0.3931 0.401 0.1693 1>}.RGB, B=@n11.R)
  )
  Add(A=@n3, B=@n12)
  n13 = ScalarParameter() {defaultValue=25, parameterName=Width}
  n14 = ScalarParameter() {defaultValue=25, parameterName=WidthTip}
  n15 = ScalarParameter() {parameterName=Use Witdh Tip}
  n16 = Transform(None=VectorParameter() {group=Rotation, parameterName=NormalizedRotationAxis, defaultValue=<rgba 1 0 0 0>}.RGB) {transformSourceType=TRANSFORMSOURCE_Instance}
  n17 = Time() {period=5}
  n18 = ScalarParameter() {defaultValue=0.2, group=Rotation, parameterName=Rotation_speed}
  n19 = VectorParameter() {group=Rotation, parameterName=PivotPointBias}
  Add(
   A=MaterialFunctionCall(
    WidthBase=@n13
    WidthTip=LinearInterpolate(A=@n13, B=@n14, Alpha=@n15)
    Expand U or V UV Channel=StaticBoolParameter() {parameterName="Expand U or V UV Channel (SB)"}
   ) {materialFunction=&SplineThicken}.WorldPositionOffset
   B=RotateAboutAxis(
    NormalizedRotationAxis=Normalize(VectorInput=@n16)
    RotationAngle=Multiply(A=@n17, B=@n18)
    PivotPoint=TransformPosition(None=@n19.RGB) {transformSourceType=TRANSFORMPOSSOURCE_Instance, transformType=TRANSFORMPOSSOURCE_World}
    Position=WorldPosition().XYZ
   )
  ) {constB=0}
  Constant() {r=20}
  Desaturation(None=@n3, Fraction=Constant() {r=-1})
  DotProduct()
  n20 = VectorParameter() {parameterName=Color_Overall, defaultValue=<rgba 1 0.0833 0.3566 1>}
  MaterialFunctionCall(Emissive=@n20.RGB) {materialFunction=&MF_FX_EmissiveAdjustments}
  Multiply(A=@n20.RGB, B=@n12)
  n21 = Add(A=@n8, B=TextureCoordinate())
  n22 = Multiply(
   A=TextureSample(UVs=Panner(Coordinate=@n21, Time=@n2) {speedX=-1.5}) {samplerType=SAMPLERTYPE_Grayscale, texture=&LowResBlurredNoise}.R
   B=@n11.R
  )
  n23 = ComponentMask(None=TextureCoordinate()) {r=true}
  n24 = Multiply(A=ScalarParameter() {defaultValue=0.8, parameterName=Opacity}, B=Multiply(A=Add(A=@n11.R, B=@n22) {constB=0}, B=Sine(None=@n23) {period=2}))
  n25 = DepthFade(Opacity=@n24, FadeDistance=ScalarParameter() {defaultValue=10, parameterName=FadeDistance}) {fadeDistanceDefault=10}
  n26 = Multiply(A=MaterialFunctionCall("FadeSensitivity (optional)"=ScalarParameter() {parameterName="FadeSensitivity (optional) (S)"}) {materialFunction=&FadeBasedOnViewAngle}.Result) {constB=10}
  n27 = ObjectPositionWS()
  n28 = CameraPositionWS()
  n29 = ScalarParameter() {defaultValue=1, group=Near Fade, parameterName=Fade_Distance}
  n30 = ScalarParameter() {group=Near Fade, parameterName=hardness}
  Multiply(
   A=Multiply(A=ShadingPathSwitch(Default=@n25, Deferred=@n25, Forward=@n24, Mobile=@n24), B=Saturate(None=@n26))
   B=OneMinus(None=SphereMask(A=@n27, B=@n28, Radius=@n29, Hardness=@n30) {hardnessPercent=10})
  )
  n31 = Transform(None=VectorParameter() {group=Rotation, parameterName=NormalizedRotationAxis, defaultValue=<rgba 1 0 0 0>}.RGB) {transformSourceType=TRANSFORMSOURCE_Instance}
  n32 = Constant() {r=0.25}
  n33 = CameraPositionWS()
  n34 = Constant3Vector() {constant=<rgba 1 1 0 0>}
  MaterialExpressionNormalize_2 = Normalize(VectorInput=Subtract(A=ObjectPositionWS(), B=Multiply(A=@n33, B=@n34)))
  n35 = Custom!(In=@MaterialExpressionNormalize_2) {
   code="float2 output;\r\nfloat2 Input1;\r\noutput = atan2 (In.y,In.x);\r\nreturn (output);"
   outputType=CMOT_Float2
   showCode=true
   inputs [{inputName=In, input {expression=&self:MaterialExpressionNormalize_2}}]
  }
  n36 = Divide(A=Divide(A=@n35, B=Constant() {r=3.142}))
  n37 = AppendVector(A=TextureCoordinate(), B=ComponentMask(None=TextureCoordinate()) {r=true})
  NamedRerouteDeclaration(None=RotateAboutAxis(
    NormalizedRotationAxis=Normalize(VectorInput=@n31)
    RotationAngle=Add(A=@n32, B=@n36)
    PivotPoint=TransformPosition(None=@n37) {transformType=TRANSFORMPOSSOURCE_World}
    Position=WorldPosition().XYZ
   )) {name=rotate with camera, nodeColor=<rgba 0.8047 0 1 1>}
  NamedRerouteUsage()
  ObjectOrientation()
  ParticleColor()
  ParticleColor()
  TextureCoordinate() {uTiling=3}
 ]
}
